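# Physics asset for the FaceAcc mesh part of the female medium LexaEarlGrey player character, set up for left and right dynamic hair chains.
# Each side links a base and four capsule bodies in series, with tight swing and twist limits, anchored to spine_05.
# Capsules also cover the upper spine, upper arms and forearms; spine_05 joins spine_04 with free rotation, while the arm bodies have no constraints.
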PhysicsAsset /Game/Characters/Player/Female/Medium/Bodies/F_MED_LexaEarlGrey/Meshes/Parts/F_MED_LexaEarlGrey_FaceAcc_Physics
bodies {
 dyn_Hair_l_01 [{
   shapeType=Capsule
   length=6.5
   radius=3.005
   shapeName=dyn_Hair_l_01_capsule
   center=<-1.4506 -1.6818 -0.6122>
   rotation=<rot 80.6045 -110.0709 20.0703>
  }]
 dyn_Hair_l_02 [{shapeType=Capsule, length=5, radius=2.505, shapeName=dyn_Hair_l_02_capsule, center=<-1.9074 -2.7136 -0.2374>, rotation=<rot 85 -90.0004 35>}]
 dyn_Hair_l_03 [{shapeType=Capsule, length=5, radius=2.505, shapeName=dyn_Hair_l_03_capsule, center=<-1.1374 -3.113 -0.2723>, rotation=<rot 85 -90.0004 20>}]
 dyn_Hair_l_04 [{shapeType=Capsule, length=5, radius=2.505, shapeName=dyn_Hair_l_04_capsule, center=<-0.2899 -3.3001 -0.2887>, rotation=<rot 85 -90.0004 5>}]
 dyn_Hair_l_base [{shapeType=Capsule, length=0.26, radius=2.25, shapeName=dyn_Hair_l_base_capsule, rotation=<rot 0 -90 0>}]
 dyn_Hair_r_01 [{
   shapeType=Capsule
   length=3
   radius=2.505
   shapeName=dyn_Hair_r_01_capsule
   center=<-1.4252 -1.7316 0.3046>
   rotation=<rot 83.0508 22.6964 152.4165>
  }]
 dyn_Hair_r_02 [{
   shapeType=Capsule
   length=3.75
   radius=2.505
   shapeName=dyn_Hair_r_02_capsule
   center=<-1.8194 -2.6643 0.6401>
   rotation=<rot 83.4425 29.29 154.0401>
  }]
 dyn_Hair_r_03 [{
   shapeType=Capsule
   length=5.5
   radius=2.005
   shapeName=dyn_Hair_r_03_capsule
   center=<-1.1374 -3.1744 0.4304>
   rotation=<rot 88.2657 -9.9632 99.9627>
  }]
 dyn_Hair_r_04 [{shapeType=Capsule, length=5.5, radius=2.005, shapeName=dyn_Hair_r_04_capsule, center=<-0.2899 -3.3001 -0.2887>, rotation=<rot 85 -90.0004 5>}]
 dyn_Hair_r_base [{shapeType=Capsule, length=0.26, radius=2.25, shapeName=dyn_Hair_r_base_capsule, rotation=<rot 0 -90 0>}]
 lowerarm_l [{
   shapeType=Capsule
   length=21.5
   radius=8.5
   shapeName=lowerarm_l_capsule
   center=<15.4871 0.0001 -2.7308>
   rotation=<rot 80.0001 -0.0008 -0.0013>
  }]
 lowerarm_r [{shapeType=Capsule, length=21.5, radius=8.5, shapeName=lowerarm_r_capsule, center=<-14.2117 0 2.5059>, rotation=<rot 80.0001 -0.0008 -0.001>}]
 spine_04 [{shapeType=Capsule, length=19.5, radius=10.5, shapeName=spine_04_capsule, center=<-0.5867 0.561 0>}]
 spine_05 [{shapeType=Capsule, length=21, radius=9, shapeName=spine_05_capsule, center=<-0.3981 -2.0489 0>}]
 upperarm_l [{shapeType=Capsule, length=25.75, radius=4.755, shapeName=upperarm_l_capsule, center=<10.5007 0 0>, rotation=<rot -0.0001 -89.9999 90>}]
 upperarm_r [{shapeType=Capsule, length=27.25, radius=5.255, shapeName=upperarm_r_capsule, center=<-11.1137 0 0>, rotation=<rot -0.0002 -89.9998 90>}]
}
constraints [
 {
  bone1Name=spine_05
  bone2Name=spine_04
  swing1LimitDegrees=45
  swing1Motion=Free
  swing2LimitDegrees=45
  swing2Motion=Free
  twistLimitDegrees=45
  twistMotion=Free
 }
 {bone1Name=dyn_Hair_r_04, bone2Name=dyn_Hair_r_03, swing1LimitDegrees=25, swing2LimitDegrees=25, twistLimitDegrees=10}
 {bone1Name=dyn_Hair_r_03, bone2Name=dyn_Hair_r_02, swing1LimitDegrees=25, swing2LimitDegrees=25, twistLimitDegrees=10}
 {bone1Name=dyn_Hair_r_02, bone2Name=dyn_Hair_r_01, swing1LimitDegrees=25, swing2LimitDegrees=25, twistLimitDegrees=10}
 {bone1Name=dyn_Hair_r_01, bone2Name=dyn_Hair_r_base, swing1LimitDegrees=25, swing2LimitDegrees=25, twistLimitDegrees=10}
 {bone1Name=dyn_Hair_l_04, bone2Name=dyn_Hair_l_03, swing1LimitDegrees=25, swing2LimitDegrees=25, twistLimitDegrees=10}
 {bone1Name=dyn_Hair_l_03, bone2Name=dyn_Hair_l_02, swing1LimitDegrees=25, swing2LimitDegrees=25, twistLimitDegrees=10}
 {bone1Name=dyn_Hair_l_02, bone2Name=dyn_Hair_l_01, swing1LimitDegrees=25, swing2LimitDegrees=25, twistLimitDegrees=10}
 {bone1Name=dyn_Hair_l_01, bone2Name=dyn_Hair_l_base, swing1LimitDegrees=25, swing2LimitDegrees=25, twistLimitDegrees=10}
 {bone1Name=dyn_Hair_r_base, bone2Name=spine_05, swing1LimitDegrees=15, swing2LimitDegrees=15, twistLimitDegrees=15}
 {bone1Name=dyn_Hair_l_base, bone2Name=spine_05, swing1LimitDegrees=15, swing2LimitDegrees=15, twistLimitDegrees=15}
]
